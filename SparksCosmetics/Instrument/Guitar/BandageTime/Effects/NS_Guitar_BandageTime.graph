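# Niagara effect for the BandageTime guitar cosmetic that creates a pair of purple glowing eyes with fire meshes, lights and sparks.
# Fire_Eyes partitions two burst particles to vary mesh tilt, visibility tag and material, while EyeGlow and pulsing Eye_Lights mirror two particles left and right.
# Sparks uses two intermittent spawn groups, one per eye, with curve-driven color, size and acceleration.
# Glow, light radius and spark size follow owner scale.
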NiagaraSystem /SparksCosmetics/Instrument/Guitar/BandageTime/Effects/NS_Guitar_BandageTime
dependencies {
 usedDataInterfaces [&NiagaraDataInterfaceArrayFloat3, &NiagaraDataInterfaceCurve, &NiagaraDataInterfaceColorCurve]
 usedDynamicInputs [&Add_Float, &MakeFloatFromVector]
 usedModules [
  &SystemState
  &self:SystemSpawnScript.SystemScriptSource.SystemScriptGraph.NiagaraNodeAssignment_0.SetVariables_#1
 ]
}
emitters {
 EyeGlow {
  emitterUpdateScript [
   &EmitterState()
   &SpawnBurst_Instantaneous(Spawn Count=2, Use Spawn Probability=false, UseLoopCountLimit=true)
  ]
  particleSpawnScript [
   &InitializeParticle(
    Color=<rgba 0.7445 0.2341 1.5 1>
    Lifetime=5
    Sprite Size Mode=::NewEnumerator3
    Uniform Sprite Size=20
    UsePositionOffset=false
    Write Material Random=true
   )
   /SetVariables(Particles.Position=&Lerp_Vector(A=<-5 0 -4>, B=<5 0 -4>, Alpha=&ReturnNormalizedExecIndex())) {isSetParameters=true}
  ]
  particleUpdateScript [
   &ParticleState(
    Kill Particles When Lifetime Has Elapsed=false
    Let Infinitely Lived Particles Die When Emitter Deactivates=true
    Loop Particles Lifetime=true
   )
   &ScaleSpriteSize(Scale Sprite Size Mode=::NewEnumerator0, Uniform Scale Factor=$System.OwnerScaleAve)
  ]
  rendererClasses [&NiagaraSpriteRendererProperties]
  renderers [{rendererClass=&NiagaraSpriteRendererProperties}]
 }
 Eye_Lights {
  emitterUpdateScript [
   &EmitterState()
   &SpawnBurst_Instantaneous(Spawn Count=2, Use Spawn Probability=false, UseLoopCountLimit=true)
  ]
  particleSpawnScript [
   &InitializeParticle(
    Color=<rgba 0.2325 0.0821 0.8 1>
    Lifetime=6
    Mesh Scale Mode=::NewEnumerator3
    Sprite Size Mode=::NewEnumerator3
    UsePositionOffset=false
    Write Material Random=true
   )
   /SetVariables(Particles.Position=&Lerp_Vector(A=<-5 2 -4>, B=<5 2 -4>, Alpha=&ReturnNormalizedExecIndex())) {isSetParameters=true}
  ]
  particleUpdateScript [
   &ParticleState(Kill Particles When Lifetime Has Elapsed=false, Let Infinitely Lived Particles Die When Emitter Deactivates=true)
   &FN_Light_Attributes(
    Color=<rgba 0.2479 0.0067 0.8 1>
    Create Color Attribute=true
    Create Exponent Attribute=true
    Create Position Attribute=true
    Create Radius Attribute=true
    Enable Particle Light Enable=true
    Exponent=2
    Light Enabled=true
    Position:NiagaraPosition=$Particles.Position
    Radius=&Multiply_Float(A=15, B=$System.OwnerScaleAve)
    Scale Color=&Lerp_Float(
     A=10
     B=20
     Alpha=&Waveform("Waveform [1]"=::NewEnumerator12, "[1] Frequency"=3, "[1] Amplitude Min/Max"=<0 1>)
    )
   )
  ]
  rendererClasses [&NiagaraLightRendererProperties, &NiagaraSpriteRendererProperties]
  renderers [
   {rendererClass=&NiagaraLightRendererProperties, rendererIndex=0}
   {rendererClass=&NiagaraSpriteRendererProperties, rendererIndex=1}
  ]
 }
 Fire_Eyes {
  emitterUpdateScript [
   &EmitterState()
   &SpawnBurst_Instantaneous(Spawn Count=2, Use Spawn Probability=false, UseLoopCountLimit=true)
  ]
  particleSpawnScript [
   &PartitionParticles()
   &InitializeParticle(
    Lifetime=6
    Mesh Scale Mode=::NewEnumerator4
    Position Offset=<403.3256 0 0>
    Sprite Size Mode=::NewEnumerator3
    UsePositionOffset=false
    Write Material Random=true
   )
   &Color(Color=&MakeLinearColorFromVectorAndFloat(
     "Vector (RGB)"=&MakeVector(
      X=&RandomRangeFloat(Minimum=0, Maximum=1)
      Y=&ReturnNormalizedExecIndex()
      Z=0
     )
     "Float (Alpha)"=1
    ))
   /SetVariables(Particles.VisibilityTag=$Particles.PartitionParticles.Partition) {isSetParameters=true}
   &InitialMeshOrientation(
    Mesh Axis To Orient=<0 0 1>
    Mesh Orientation Mode=::NewEnumerator2
    Orientation Vector=<0 0 1>
    Rotation=&SelectVectorFromArray(
     Direct Array Index=$Particles.PartitionParticles.Partition
     Vector Selection Array=DataInterface() {propertyValues {FloatData [<-0.1691 0 -0.02>, <-0.1691 0 0.02>]}}
    )
    Use Rotation Vector=true
   )
   /SetVariables(Particles.LeftEye_Position:NiagaraPosition=<-3 1 -2.5>, Particles.RightEye_Position:NiagaraPosition=<3 1 -2.5>) {isSetParameters=true}
   &DynamicMaterialParameters(
    Index 0 Param 1=&Lerp_Float(A=0, B=0.3, Alpha=&MakeFloatFromInt(Int=$Particles.VisibilityTag))
    Module.Index0Param0Write=true
    Module.Index0Param1Write=false
    Module.Index0Param2Write=false
    Module.Index0Param3Write=false
   ) {module=DynamicMaterialParameters001}
  ]
  particleUpdateScript [&ParticleState(Kill Particles When Lifetime Has Elapsed=false, Let Infinitely Lived Particles Die When Emitter Deactivates=true)]
  rendererClasses [&NiagaraMeshRendererProperties]
  renderers [
   {rendererClass=&NiagaraMeshRendererProperties, rendererIndex=0}
   {rendererClass=&NiagaraMeshRendererProperties, rendererIndex=1}
  ]
 }
 Sparks {
  emitterUpdateScript [
   &EmitterState()
   &SpawnRate(
    Spawn Probability=0.3
    SpawnRate=&RandomRangeFloat(Minimum=0, Maximum=6, Evaluation Type=::NewEnumerator1)
    Use Spawn Probability=true
   )
   &SpawnRate(
    Spawn Probability=0.3
    SpawnGroup=1
    SpawnRate=&RandomRangeFloat(Minimum=0, Maximum=6, Evaluation Type=::NewEnumerator1)
    Use Spawn Probability=true
   ) {module=SpawnRate001}
  ]
  particleSpawnScript [
   /SetVariables(Particles.Direction:NiagaraInt32=&SpawnGroup()) {isSetParameters=true}
   &InitializeParticle(
    Lifetime=&RandomRangeFloat(Minimum=2, Maximum=2)
    Position Offset=&Lerp_Vector(A=<-5 0 -4>, B=<5 0 -4>, Alpha=&MakeFloatFromInt(Int=$Particles.Direction))
    Sprite Size Mode=::NewEnumerator3
    Uniform Sprite Size=&RandomRangeFloat(Minimum=4, Maximum=8)
    UsePositionOffset=true
    Write Material Random=true
   )
   &AddVelocity(
    Cone Angle=30
    Cone Axis=&Lerp_Vector(A=<-1 0.6 0>, B=<1 0.6 0>, Alpha=&MakeFloatFromInt(Int=$Particles.Direction))
    Module.UseConeFalloff=false
    Module.UseVelDistribution=false
    Override Seed=false
    Velocity Mode=::NewEnumerator2
    Velocity Speed=20
   )
   &AddVelocity(Velocity=&RandomRangeVector(Minimum=<0 0 5>, Maximum=<0 0 10>)) {module=AddVelocity001}
  ]
  particleUpdateScript [
   &ParticleState()
   &ScaleSpriteSize(
    Uniform Curve Scale=$System.OwnerScaleAve
    Uniform Curve Sprite Scale=DataInterface() {propertyValues {Curve {keys [
        {interpMode=RCIM_Cubic, tangentMode=RCTM_User, time=0, value=1}
        {arriveTangent=-0.0087, interpMode=RCIM_Cubic, leaveTangent=-0.0087, tangentMode=RCTM_User, time=1, value=0}
       ]}}}
   )
   &CurlNoiseForce(Noise Frequency=200, Noise Strength=30, Pan Noise=true, Pan Noise Field=<0 0 2>) {enabled=false}
   &Drag(Use Linear Drag=true, Use Rotational Drag=true)
   &AccelerationForce(
    Acceleration=&Multiply_VectorByFloat(
     Vector=<0 0 15>
     Float=&FloatFromCurve(FloatCurve=DataInterface() {propertyValues {Curve {keys [<at 0.3 1>,<at 0.5 -2>]}}})
    )
    Coordinate Space=::Local
   )
   &SolveForcesAndVelocity(Clamp Velocity=false, Limit Acceleration=false)
   &ScaleSpriteSizeBySpeed(Max Scale Factor=<1 2>, Min Scale Factor=<0.3 1>)
   &Color(Color=&ColorFromCurve(ColorCurve=DataInterface() {propertyValues {
       AlphaCurve {keys [<at 0 0.3 RCIM_Cubic>,<at 0.2766 0.25>,<at 1 0 RCIM_Cubic>]}
       BlueCurve {keys [<at 0 1 RCIM_Cubic>,<at 0.3172 0.2>,<at 0.6657 0.2 RCIM_Cubic>]}
       GreenCurve {keys [<at 0 0.176 RCIM_Cubic>,<at 0.3172 0.0119>,<at 0.6657 0 RCIM_Cubic>]}
       RedCurve {keys [<at 0 0.7468 RCIM_Cubic>,<at 0.3172 0.1457>,<at 0.6657 0.0569 RCIM_Cubic>]}
      }}))
   &DynamicMaterialParameters(
    Index 0 Param 1=&FloatFromCurve(FloatCurve=DataInterface() {propertyValues {Curve {keys [{arriveTangent=5.579, interpMode=RCIM_Cubic, leaveTangent=5.579, tangentMode=RCTM_User, time=0, value=-2}, <at 1 1>]}}})
    Index 0 Param 2=&RandomRangeFloat(Minimum=0, Maximum=1)
    Index 0 Param 3=&RandomRangeFloat(Minimum=0, Maximum=1)
    Module.Index0Param0Write=true
    Module.Index0Param1Write=true
    Module.Index0Param2Write=true
    Module.Index0Param3Write=false
   )
  ]
  rendererClasses [&NiagaraSpriteRendererProperties]
  renderers [{rendererClass=&NiagaraSpriteRendererProperties}]
 }
}
summary {
 systemName=NS_Guitar_BandageTime
 emitters [
  {emitterName=Fire_Eyes, rendererClasses [&NiagaraMeshRendererProperties]}
  {emitterName=Eye_Lights, rendererClasses [&NiagaraLightRendererProperties, &NiagaraSpriteRendererProperties]}
  {emitterName=Sparks, rendererClasses [&NiagaraSpriteRendererProperties]}
  {emitterName=EyeGlow, rendererClasses [&NiagaraSpriteRendererProperties]}
 ]
 userVariables []
}
