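# Physics asset for the BandageTime guitar cosmetic in SparksCosmetics, providing capsule bodies and joints for its dangling cloth and ribbon bones.
# The multi-capsule handle body anchors three chains: left cloth, right cloth and a three-bone ribbon.
# Both cloth chains meet at cloth_l_04 through fully angle-locked joints, closing a loop from the handle.
# Most cloth joints lock twist and vary swing limits, while ribbon joints allow only slight twist.
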PhysicsAsset /SparksCosmetics/Instrument/Guitar/BandageTime/Meshes/BandageTime_Guitar_Physics
bodies {
 cloth_03 [{shapeType=Capsule, length=3, radius=2, shapeName=cloth_03_capsule, center=<1.3465 -0.5696 -0.238>, rotation=<rot 0 -107.0876 -69.6594>}]
 cloth_l_01 [{shapeType=Capsule, length=2, radius=2, shapeName=cloth_l_01_capsule, center=<1.4631 0.9159 -0.4509>, rotation=<rot 0.0001 58.7352 -87.3496>}]
 cloth_l_02 [{shapeType=Capsule, length=2, radius=2, shapeName=cloth_l_02_capsule, center=<1.2235 0.026 -0.0791>, rotation=<rot -0.0001 -121.6606 -83.664>}]
 cloth_l_04 [{
   shapeType=Capsule
   length=3
   radius=2
   shapeName=cloth_l_04_capsule
   center=<0.3576 -0.1921 -0.1028>
   rotation=<rot -0.8615 -102.6326 -91.7853>
  }]
 cloth_r_01 [{shapeType=Capsule, length=1.7699, radius=2.1705, shapeName=cloth_r_01_capsule, center=<1 -2 0>, rotation=<rot 0.0001 -88.3854 -80.4212>}]
 cloth_r_02 [{shapeType=Capsule, length=3.9824, radius=1.8991, shapeName=cloth_r_02_capsule, center=<0.5 -1 0.3>, rotation=<rot 1.6655 82.804 -94.5853>}]
 cloth_r_03 [{shapeType=Capsule, length=3, radius=2, shapeName=cloth_r_03_capsule, center=<0.7772 -0.121 0.0259>, rotation=<rot -0.0001 81.1477 -88.1145>}]
 cloth_r_connect [{
   shapeType=Capsule
   length=3
   radius=2
   shapeName=cloth_r_connect_capsule
   center=<1.0632 0.3367 0.2271>
   rotation=<rot 0.0004 -90.0002 80.0001>
  }]
 handle [
  {shapeType=Capsule, length=31.9067, radius=3.826, shapeName=handle_capsule, center=<-1.1921 -1.1514 -15.3641>, rotation=<rot 0.0003 0 -90>}
  {shapeType=Capsule, length=13.3352, radius=4.1941, shapeName=handle_capsule, center=<-2.3394 -0.408 -19.3808>, rotation=<rot 0.0003 0 -90>}
  {shapeType=Capsule, length=11.7972, radius=3.0058, shapeName=handle_capsule, center=<-0.9976 -1.9987 67.1094>, rotation=<rot 0 -90 0>}
 ]
 ribbon_01 [{shapeType=Capsule, length=1.4546, radius=2.2124, shapeName=ribbon_01_capsule, center=<0 -0.2651 0>, rotation=<rot 90 0 0>}]
 ribbon_02 [{
   shapeType=Capsule
   length=1.6506
   radius=2.4135
   shapeName=ribbon_02_capsule
   center=<-0.0047 -0.0037 -0.1156>
   rotation=<rot -0.0249 -90 90>
  }]
 ribbon_03 [{shapeType=Capsule, length=0.4849, radius=3.0424, shapeName=ribbon_03_capsule, center=<2.2865 1 0.5559>, rotation=<rot -0.2904 -90 90>}]
}
constraints [
 {bone1Name=cloth_l_01, bone2Name=handle, swing1LimitDegrees=15, swing2LimitDegrees=45, twistLimitDegrees=45, twistMotion=Locked}
 {bone1Name=cloth_l_02, bone2Name=cloth_l_01, swing1LimitDegrees=45, swing2LimitDegrees=90, twistLimitDegrees=10}
 {bone1Name=cloth_03, bone2Name=cloth_l_02, swing1LimitDegrees=45, swing2LimitDegrees=90, twistLimitDegrees=45, twistMotion=Locked}
 {
  bone1Name=cloth_l_04
  bone2Name=cloth_03
  swing1LimitDegrees=45
  swing1Motion=Locked
  swing2LimitDegrees=90
  swing2Motion=Locked
  twistLimitDegrees=45
  twistMotion=Locked
 }
 {bone1Name=cloth_r_01, bone2Name=handle, swing1LimitDegrees=5, swing2LimitDegrees=45, twistLimitDegrees=45, twistMotion=Locked}
 {bone1Name=cloth_r_02, bone2Name=cloth_r_01, swing1LimitDegrees=10, swing2LimitDegrees=55, twistLimitDegrees=45, twistMotion=Locked}
 {bone1Name=cloth_r_03, bone2Name=cloth_r_02, swing1LimitDegrees=45, swing2LimitDegrees=90, twistLimitDegrees=45, twistMotion=Locked}
 {bone1Name=cloth_r_connect, bone2Name=cloth_r_03, swing1LimitDegrees=45, swing2LimitDegrees=90, twistLimitDegrees=45, twistMotion=Locked}
 {bone1Name=ribbon_01, bone2Name=handle, swing1LimitDegrees=90, swing2LimitDegrees=90, twistLimitDegrees=3}
 {bone1Name=ribbon_02, bone2Name=ribbon_01, swing1LimitDegrees=60, swing2LimitDegrees=80, twistLimitDegrees=5}
 {bone1Name=ribbon_03, bone2Name=ribbon_02, swing1LimitDegrees=45, swing2LimitDegrees=30, twistLimitDegrees=5}
 {
  bone1Name=cloth_l_04
  bone2Name=cloth_r_connect
  swing1LimitDegrees=45
  swing1Motion=Locked
  swing2LimitDegrees=90
  swing2Motion=Locked
  twistLimitDegrees=45
  twistMotion=Locked
 }
]
